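# Master masked, lit material for waterfall foam, rendering mesh cards as camera-facing sprites.
# World position offset combines distance-based sprite scaling, sine drift, looping directional travel, world-aligned noise and camera offset.
# A soft smoke mask broken by two panning noises fades with the travel cycle, with optional sphere masking.
# Sphere and water normals add shading, time-of-day functions drive color, and dithered pixel depth offset applies outside forward shading and low quality.
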Material /Game/Effects/Fort_Effects/Materials/Base/M_Apollo_Master_Waterfall_Foam_NoTick
graph {
 properties {
  bUsedWithMeshParticles=true
  bUsedWithNiagaraMeshParticles=true
  blendMode=BLEND_Masked
  pixelDepthOffsetMode=PDOM_Legacy
  translucencyLightingMode=TLM_VolumetricPerVertexNonDirectional
  usedShadingModels=Default Lit
 }
 nodes [
  n1 = Multiply(A=Time(), B=ScalarParameter() {defaultValue=0.001, group=WPO, parameterName=VarianceSpeed}) {constB=0.05}
  n2 = ComponentMask(None=TextureCoordinate() {coordinateIndex=1}) {r=true}
  n3 = Add(
   A=Multiply(A=Time(), B=ScalarParameter() {defaultValue=0.4, group=WPO, parameterName=DriftSpeed})
   B=Sine(None=Add(A=@n1, B=@n2)) {period=0.17}
  )
  n4 = Sine(None=@n3) {period=0.56}
  n5 = Sine(None=@n3) {period=0.7}
  n6 = Add(A=@n1, B=@n2)
  n7 = Multiply(
   A=AppendVector(A=AppendVector(A=@n4, B=@n5), B=Sine(None=@n3) {period=0.82})
   B=LinearInterpolate(
    A=ScalarParameter() {defaultValue=1, group=WPO, parameterName=MinHorizontalDrift}
    B=ScalarParameter() {defaultValue=1, group=WPO, parameterName=MaxHorizontalDrift}
    Alpha=Sine(None=@n6) {period=0.14}
   )
  )
  n8 = Multiply(A=ObjectRadius(), B=ScalarParameter() {defaultValue=1, group=WPO, parameterName=ObjectRadius_Mult}) {constB=0.5}
  n9 = Multiply(A=@n8) {constB=-1}
  n10 = Divide(A=Time() {bOverride_Period=true, period=1}) {constB=1}
  n11 = Divide(A=Time() {bOverride_Period=true, period=2})
  n12 = TextureCoordinate() {coordinateIndex=1}
  n13 = Multiply(A=Subtract(A=ComponentMask(None=@n12) {g=true}) {constB=-0.5}) {constB=-2}
  n14 = StaticSwitchParameter(True=Add(A=StaticSwitchParameter(True=@n10, False=@n11) {defaultValue=true, parameterName=UseFastSprites}, B=@n13), False=@n13) {defaultValue=true, parameterName=Continuous Up and down scaling?}
  n15 = Frac(None=@n14)
  n16 = Multiply(
   A=Multiply(A=LinearInterpolate(A=@n9, B=@n8, Alpha=@n15), B=ScalarParameter() {defaultValue=1, group=WPO, parameterName=OverallSpeed})
   B=VectorParameter() {parameterName=OverallDirection, defaultValue=<rgba 1 0 0 1>}.RGB
  )
  n17 = MaterialFunctionCall(Distance=ScalarParameter() {defaultValue=2500, parameterName=SoftenDistance}) {materialFunction=&Near_Camera_Fade}
  n18 = Multiply(
   A=VertexColor().R
   B=LinearInterpolate(
    A=ScalarParameter() {defaultValue=150, parameterName=SpriteScaleNear}
    B=ScalarParameter() {defaultValue=300, parameterName=SpriteScaleFar}
    Alpha=@n17.Result
   )
  )
  n19 = TextureSample() {samplerType=SAMPLERTYPE_Normal, texture=&T_Sphere_N}
  n20 = Constant3Vector() {constant=<rgba 0 0 1 0>}
  n21 = ScalarParameter() {parameterName=LerpNormal}
  n22 = TextureCoordinate() {vTiling=2}
  n23 = VertexColor()
  n24 = LinearInterpolate(
   A=ScalarParameter() {defaultValue=1, parameterName=TilingDistanceNear_Mult}
   B=ScalarParameter() {defaultValue=1, parameterName=TilingDistanceFar_Mult}
   Alpha=MaterialFunctionCall(Distance=ScalarParameter() {defaultValue=2500, parameterName=TilingDistance}) {materialFunction=&Near_Camera_Fade}.Result
  )
  n25 = Multiply(
   A=LinearInterpolate(
    A=ScalarParameter() {defaultValue=1, parameterName=ScaleOpacityMaskNoise}
    B=ScalarParameter() {defaultValue=1, parameterName=ScaleOpacityMaskNoise_Max}
    Alpha=@n23.R
   )
   B=@n24
  )
  n26 = Panner(
   Coordinate=Add(A=Multiply(A=@n22, B=@n25), B=AppendVector(A=@n23.R, B=@n23.R))
   Speed=LinearInterpolate(
    A=VectorParameter() {parameterName=DropSpeed, defaultValue=<rgba 0 -2 0 1>}.RGB
    B=VectorParameter() {parameterName=DropSpeedMax, defaultValue=<rgba 0 -2 0 1>}.RGB
    Alpha=VertexColor().R
   )
  ) {speedY=-0.5}
  n27 = TextureSample(UVs=@n26) {samplerType=SAMPLERTYPE_Normal, texture=/Game/Packages/DS_Fortnite_TRV_Terrain/Textures/T_TRV_SwampWater_01_N}
  n28 = ScalarParameter() {defaultValue=0.7, parameterName=Flatten}
  n29 = MaterialFunctionCall!(
   Center Location=WorldPosition().XYZ
   0-1 UVs
   XY scale=AppendVector(A=@n18, B=@n18)
   OverridePivot
   "Normals (Optional)"=MaterialFunctionCall(
    BaseNormal=LinearInterpolate(A=@n19.RGB, B=@n20, Alpha=@n21)
    AdditionalNormal=MaterialFunctionCall(Normal=@n27.RGB, Flatness=@n28) {materialFunction=&FlattenNormal}.Result
   ) {materialFunction=&BlendAngleCorrectedNormals}.Result
  ) {materialFunction=&Sprite}
  n30 = VectorParameter() {parameterName=OffsetPos, defaultValue=<rgba 0 0 -15 0>}
  n31 = VectorParameter() {parameterName=Offset_1, defaultValue=<rgba 0 0 60 0>}
  n32 = WorldPosition()
  n33 = VectorParameter() {parameterName=WorldMaskTexture_Offset, defaultValue=<rgba 0.005 0.005 0 0>}
  n34 = VectorParameter() {parameterName=WorldMaskTexture_Offset_Max, defaultValue=<rgba 0.005 0.005 0 0>}
  n35 = VertexColor()
  n36 = Multiply(
   A=Multiply(A=Time(), B=LinearInterpolate(A=@n33.RGB, B=@n34.RGB, Alpha=@n35.R))
   B=ScalarParameter() {defaultValue=0.9, parameterName=Noisespeed}
  )
  n37 = ComponentMask(None=MaterialFunctionCall(
    TextureObject=TextureObject() {texture=&T_Inky_Smoke_Tile}
    TextureSize=VectorParameter() {parameterName=Size, defaultValue=<rgba 400 400 400 0>}.RGB
    WorldPosition=Multiply(A=@n32.XYZ, B=@n36)
   ) {materialFunction=&WorldAlignedTexture}.XYZ Texture) {g=true}
  n38 = Transform(None=Multiply(A=VertexColor().R, B=LinearInterpolate(A=@n30.RGB, B=@n31.RGB, Alpha=@n37))) {transformSourceType=TRANSFORMSOURCE_Local}
  Add(
   A=MaterialFunctionCall(Offset Amount=ScalarParameter() {defaultValue=10, group=WPO, parameterName=CameraOffset}) {materialFunction=/Engine/Functions/Engine_MaterialFunctions02/WorldPositionOffset/CameraOffset}.Clamped Camera Offset
   B=Add(A=Add(A=@n29.World position offset, B=@n38), B=Add(A=@n7, B=@n16))
  )
  n39 = MaterialFunctionCall() {materialFunction=/Game/Athena/Items/Gameplay/SilkyVeteran/FX/SlurpVatMask}
  n40 = TextureSample() {texture=/Game/Effects/Fort_Effects/Textures/T_Smoke_Soft}
  n41 = TextureSample(UVs=@n26) {texture=/Game/Global/Textures/Default/Noise/TankMoteNoise}
  n42 = ScalarParameter() {defaultValue=1, parameterName=ScaleOpacityMaskNoise}
  n43 = ScalarParameter() {defaultValue=1, parameterName=ScaleOpacityMaskNoise_Max}
  n44 = VertexColor()
  n45 = Multiply(A=TextureCoordinate() {uTiling=1.2, vTiling=1.6}, B=Multiply(A=@n24, B=LinearInterpolate(A=@n42, B=@n43, Alpha=@n44.A)))
  n46 = AppendVector(A=@n44.G, B=@n44.G)
  n47 = VectorParameter() {parameterName=DropSpeed2, defaultValue=<rgba 1 -2 0 1>}
  n48 = VectorParameter() {parameterName=DropSpeedMax2, defaultValue=<rgba 1 -2 0 1>}
  n49 = VertexColor()
  n50 = TextureSample(UVs=Panner(Coordinate=Add(A=@n45, B=@n46), Speed=LinearInterpolate(A=@n47.RGB, B=@n48.RGB, Alpha=@n49.G)) {speedY=-0.5}) {texture=&TankMoteNoise}
  n51 = LinearInterpolate(A=@n40.G, B=Multiply(A=@n40.G, B=Add(A=@n41.G, B=@n50.G)), Alpha=ScalarParameter() {defaultValue=0.5, parameterName=LErpoit})
  n52 = Multiply(A=@n51, B=LinearInterpolate(A=ScalarParameter() {defaultValue=0.7, parameterName=OpacityMask}, Alpha=@n17.Result) {constA=0.5, constB=0.9})
  n53 = Sine(None=Add(A=@n14) {constB=-0.25})
  Multiply(
   A=StaticSwitchParameter(True=Multiply(A=@n52, B=@n39.Result), False=@n52) {parameterName=Mask to World SphereMask}
   B=LinearInterpolate(Alpha=ConstantBiasScale(None=@n53))
  )
  n54 = MaterialFunctionCall(
   Value_Morning=ScalarParameter() {defaultValue=1, group=Opacity, parameterName="Value_Morning (S)"}
   Value_Day=ScalarParameter() {defaultValue=1, group=Opacity, parameterName="Value_Day (S)"}
   Value_Evening=ScalarParameter() {defaultValue=1, group=Opacity, parameterName="Value_Evening (S)"}
   Value_Night=ScalarParameter() {defaultValue=1, group=Opacity, parameterName="Value_Night (S)"}
  ) {materialFunction=&Scalar_Time_of_Day}
  Multiply(
   A=MaterialFunctionCall(
    Color_Morning=VectorParameter() {group=TOD, parameterName="Color_Morning (V3)", defaultValue=<rgba 1 1 1 0>}.RGB
    Color_Day=VectorParameter() {group=TOD, parameterName="Color_Day (V3)", defaultValue=<rgba 1 1 1 0>}.RGB
    Color_Evening=VectorParameter() {group=TOD, parameterName="Color_Evening (V3)", defaultValue=<rgba 1 1 1 0>}.RGB
    Color_Night=VectorParameter() {group=TOD, parameterName="Color_Night (V3)", defaultValue=<rgba 1 1 1 0>}.RGB
   ) {materialFunction=&Color_Time_of_Day}.Result
   B=@n54.Result
  )
  Multiply(B=@n54.Result)
  n55 = Constant()
  n56 = ScalarParameter() {defaultValue=110, parameterName=PixelDepthOffset}
  n57 = ShadingPathSwitch(
   Default=Multiply(
    A=Multiply(A=@n51, B=@n56)
    B=MaterialFunctionCall() {materialFunction=&DitherTemporalAA}.Result
   )
   Forward=@n55
  )
  QualitySwitch(Default=@n55, High=@n57, Medium=@n57)
 ]
}
