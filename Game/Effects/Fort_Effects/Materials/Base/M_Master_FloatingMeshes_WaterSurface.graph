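# Master masked, two-sided mesh-particle material for small debris, such as leaves or trash, floating on water surfaces.
# World position offset combines axis rotation, optional looping drift with matching opacity fade, camera depth-fade offsets, optional water displacement, curl noise and player bending.
# Low quality, forward and mobile paths skip curl noise and the camera sphere-mask push.
# Color and opacity can use textures or animated SubUV flipbooks, with vertex color varying tint and speeds.
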Material /Game/Effects/Fort_Effects/Materials/Base/M_Master_FloatingMeshes_WaterSurface
graph {
 properties {
  bUsedWithMeshParticles=true
  bUsedWithNiagaraMeshParticles=true
  blendMode=BLEND_Masked
  pixelDepthOffsetMode=PDOM_Legacy
  translucencyLightingMode=TLM_Surface
  twoSided=true
  usedShadingModels=Default Lit
 }
 nodes [
  n1 = Time()
  n2 = ObjectPositionWS()
  n3 = Add(A=@n1, B=ComponentMask(None=@n2) {b=true})
  n4 = ScalarParameter() {defaultValue=0.07, group=WPO Rotation, parameterName=RotationSpeed}
  n5 = ScalarParameter() {defaultValue=0.05, group=WPO Rotation, parameterName=RotationSpeedReverse}
  n6 = VertexColor()
  n7 = RotateAboutAxis(
   NormalizedRotationAxis=Transform(None=VectorParameter() {parameterName=RotationAxis, defaultValue=<rgba 0 0 1 0>}.RGB) {transformSourceType=TRANSFORMSOURCE_Local}
   RotationAngle=Multiply(A=StaticSwitchParameter(True=@n3, False=@n1) {parameterName=UsewithGrassActor?}, B=LinearInterpolate(A=@n4, B=@n5, Alpha=@n6.R))
   PivotPoint=StaticSwitchParameter(True=@n2, False=ActorPositionWS()) {parameterName=UsewithGrassActor?}
   Position=WorldPosition().XYZ
  )
  n8 = Multiply(A=Time(), B=ScalarParameter() {defaultValue=0.001, group=WPO, parameterName=VarianceSpeed}) {constB=0.05}
  n9 = ComponentMask(None=TextureCoordinate() {coordinateIndex=1}) {r=true}
  n10 = Add(
   A=Multiply(A=Time(), B=ScalarParameter() {defaultValue=0.4, group=WPO, parameterName=DriftSpeed})
   B=Sine(None=Add(A=@n8, B=@n9)) {period=0.17}
  )
  n11 = AppendVector(A=Sine(None=@n10) {period=0.56}, B=Sine(None=@n10) {period=0.7})
  n12 = Sine(None=@n10) {period=0.82}
  n13 = ScalarParameter() {defaultValue=1, group=WPO, parameterName=MinHorizontalDrift}
  n14 = ScalarParameter() {defaultValue=1, group=WPO, parameterName=MaxHorizontalDrift}
  n15 = Sine(None=Add(A=@n8, B=@n9)) {period=0.14}
  n16 = Multiply(A=ObjectRadius(), B=ScalarParameter() {defaultValue=1, parameterName=ObjectRadiusMult}) {constB=0.5}
  n17 = Time() {bOverride_Period=true, period=6}
  n18 = Time() {bOverride_Period=true, period=12}
  n19 = ComponentMask(None=TextureCoordinate() {coordinateIndex=1}) {g=true}
  n20 = ScalarParameter() {defaultValue=0.5, parameterName=ObjectUVSubtract}
  n21 = Add(
   A=StaticSwitchParameter(True=Divide(A=@n17) {constB=6}, False=Divide(A=@n18) {constB=12}) {defaultValue=true, parameterName=UseFastSprites}
   B=Multiply(A=Subtract(A=@n19, B=@n20) {constB=0.5}) {constB=-2}
  )
  n22 = LinearInterpolate(A=Multiply(A=@n16) {constB=-1}, B=@n16, Alpha=Frac(None=@n21))
  n23 = ScalarParameter() {defaultValue=1, group=WPO, parameterName=OverallSpeed}
  n24 = Add(
   A=Multiply(A=AppendVector(A=@n11, B=@n12), B=LinearInterpolate(A=@n13, B=@n14, Alpha=@n15))
   B=Multiply(A=Multiply(A=@n22, B=@n23), B=VectorParameter() {parameterName=OverallDirection, defaultValue=<rgba 1 0 0 1>}.RGB)
  )
  n25 = StaticSwitchParameter(True=Transform(None=@n24), False=@n24) {parameterName=UseWorldSpaceMovement?}
  n26 = Multiply(
   A=MaterialFunctionCall(
    Fade Length=ScalarParameter() {defaultValue=0.4, group=WPO, parameterName=CameraFade_Length, sliderMax=10000}
    Fade Offset=ScalarParameter() {defaultValue=5000, group=WPO, parameterName=CameraFade_Offset, sliderMax=10000}
    For Vertex Shader=StaticBool() {value=true}
   ) {materialFunction=&CameraDepthFade}.Result
   B=VectorParameter() {group=HeightOffset, parameterName=Z_Push, defaultValue=<rgba 0 0 300 0>}.RGB
  )
  n27 = Multiply(
   A=MaterialFunctionCall!(
    Ignore XY Offsets
    Position
    WaterBodyIndex=ScalarParameter() {group=WPO, parameterName=WaterBodyIndex}
    Depth=ScalarParameter() {group=WPO, parameterName=WaterDepth}
   ) {materialFunction=/FortWater/Materials/Functions/WaterParticleDisplacement}.WPO
   B=VectorParameter() {parameterName=Water Displacement Multiplier, defaultValue=<rgba 1 1 1 0>}.RGB
  )
  n28 = ParticlePositionWS()
  n29 = ObjectPositionWS()
  n30 = ComponentMask(None=Subtract(A=CameraPositionWS(), B=FeatureLevelSwitch(Default=@n28, ES2=@n29, ES3_1=@n29, SM4_REMOVED=@n28))) {b=true}
  n31 = MaterialFunctionCall(Z=Multiply(A=ScalarParameter() {defaultValue=1, group=WPO, parameterName=ZOffset}, B=Clamp(None=@n30) {minDefault=-1})) {materialFunction=&MakeFloat3}
  n32 = Add(
   A=MaterialFunctionCall(Offset Amount=ScalarParameter() {defaultValue=10, group=WPO, parameterName=CameraOffset}) {materialFunction=/Engine/Functions/Engine_MaterialFunctions02/WorldPositionOffset/CameraOffset}.Clamped Camera Offset
   B=StaticSwitchParameter(True=Add(A=@n27, B=@n31.Result), False=Constant()) {parameterName=UseWaterDisplacement}
  )
  n33 = Add(
   A=StaticSwitchParameter(True=@n7, False=Add(A=@n7, B=@n25)) {defaultValue=true, parameterName=UseRotation}
   B=StaticSwitchParameter(True=Add(A=@n26, B=@n32), False=Constant()) {defaultValue=true, parameterName=UseCameraOffset?}
  )
  n34 = VertexColor()
  n35 = Multiply(
   A=LinearInterpolate(
    A=ScalarParameter() {defaultValue=50, group=WPO Curl, parameterName=CurIntensity}
    B=ScalarParameter() {defaultValue=100, group=WPO Curl, parameterName=CurIntensityHigh}
    Alpha=@n34.A
   )
   B=VertexColor().A
  )
  n36 = Subtract(A=WorldPosition().XYZ, B=ActorPositionWS())
  n37 = Multiply(A=Time(), B=ScalarParameter() {defaultValue=1, group=WPO Curl, parameterName=CurlSpeed})
  n38 = ScalarParameter() {defaultValue=1000, group=WPO Curl, parameterName=CurlScale}
  n39 = ScalarParameter() {defaultValue=500, group=WPO Curl, parameterName=CurlScale_Low}
  n40 = VectorNoise(World Position=Divide(A=Subtract(A=@n36, B=@n37), B=LinearInterpolate(A=@n38, B=@n39, Alpha=@n34.A))) {noiseFunction=VNF_VectorALU}
  n41 = SphereMask(
   A=CameraPositionWS()
   B=WorldPosition().XYZ
   Radius=ScalarParameter() {defaultValue=200, group=WPO Player Interact, parameterName=camerafadeMaskRadius}
   Hardness=ScalarParameter() {group=WPO Player Interact, parameterName=camerafadeMaskHardness}
  )
  n42 = Subtract(A=WorldPosition().XYZ, B=CameraPositionWS())
  n43 = Constant3Vector() {constant=<rgba 1 1 0 0>}
  n44 = Add(
   A=Multiply(A=Multiply(A=@n42, B=@n43), B=ScalarParameter() {defaultValue=1, group=WPO Player Interact, parameterName=CameraFadeMaskPush})
   B=AppendVector(A=Constant2Vector(), B=ScalarParameter() {defaultValue=-200, group=WPO Player Interact, parameterName=CameraFadeMaskSink})
  )
  n45 = Add(
   A=Add(A=@n33, B=Multiply(A=@n35, B=@n40))
   B=StaticSwitchParameter(True=Multiply(A=@n41, B=@n44), False=Constant()) {defaultValue=true, parameterName=UsePlayerOffset?}
  )
  n46 = FeatureLevelSwitch(Default=@n45, ES2=@n33, ES3_1=@n33, SM4_REMOVED=@n45, SM5=@n45, SM6=@n45)
  n47 = CollectionParameter() {collection=&FortniteMaterialParameters, parameterName=PlayerPosition}
  n48 = MaterialFunctionCall!(
   PlayerPosition=@n47
   PlayerRadiusMax=ScalarParameter() {defaultValue=2500, group=WPO Player Interact, parameterName=WPO_PlayerRadius}
   BendStrength=ScalarParameter() {defaultValue=25, group=WPO Player Interact, parameterName=WPO_Bend}
  ) {materialFunction=/Game/Athena/Environments/Landscape/MaterialFunctions/MF_AthenaFoliagePlayerInteract}
  n49 = Multiply(A=@n48.Rotation, B=Constant() {r=1})
  Add(
   A=QualitySwitch(Default=ShadingPathSwitch(Default=@n46, Forward=@n33), Low=@n33)
   B=StaticSwitchParameter(True=LinearInterpolate(B=@n49, Alpha=@n48.PlayerSphereMask), False=Constant()) {defaultValue=true, parameterName=USePlayerOffset?}
  )
  LinearInterpolate()
  n50 = VectorParameter() {group=Color, parameterName=BaseColor_Secondary, defaultValue=<rgba 0.125 0.2182 0.25 1>}
  n51 = VectorParameter() {group=Color, parameterName=BaseColor, defaultValue=<rgba 0.125 0.2182 0.25 1>}
  n52 = VertexColor()
  n53 = MaterialFunctionCall!(
   U_Images=ScalarParameter() {defaultValue=8, parameterName="U_Images (S)"}
   V_Images=ScalarParameter() {defaultValue=8, parameterName="V_Images (S)"}
   SubUV Texture RGB=TextureObjectParameter() {parameterName="SubUV Texture RGB (T2d)", texture=/Game/Effects/Fort_Effects/Textures/Leaf/T_Leaf_SubUV}
   Time Speed=LinearInterpolate(
    A=ScalarParameter() {defaultValue=3, group=SubUV, parameterName=Time Speed}
    B=ScalarParameter() {defaultValue=3, group=SubUV, parameterName=Time Speed Slow}
    Alpha=VertexColor().R
   )
   Manual Time Control
   Time
   UVs
  ) {materialFunction=/Game/Effects/Fort_Effects/Materials/_Functions/MF_SubUV}
  Multiply(
   A=LinearInterpolate(
    A=@n51.RGB
    B=LinearInterpolate(A=@n50.RGB, B=@n51.RGB, Alpha=@n52.R)
    Alpha=ScalarParameter() {group=Color, parameterName=UseSecondaryColor?}
   )
   B=StaticSwitchParameter(
    True=ComponentMask(None=@n53.RGB) {g=true}
    False=TextureSampleParameter2D() {group=Textures, parameterName=Diffuse Texture, texture=/Game/Environments/Props/Trash/Textures/Prop_Trash_D}.RGB
   ) {parameterName=UseSubUV?}
  )
  n54 = TextureSampleParameter2D() {
   group=Textures
   parameterName=Spec Texture
   samplerType=SAMPLERTYPE_Masks
   sortPriority=1
   texture=/Game/Environments/Props/Trash/Textures/Prop_Trash_S
  }
  Multiply(A=ScalarParameter() {defaultValue=1, parameterName=Specular}, B=@n54.B)
  Multiply(A=ScalarParameter() {defaultValue=1, parameterName=Roughness}, B=@n54.R)
  n55 = TextureSampleParameter2D(UVs=Multiply(A=TextureCoordinate(), B=ScalarParameter() {defaultValue=1, group=Textures, parameterName=MaskTextureTiling})) {
   desc=Use green channel
   group=Textures
   parameterName=Opacity Mask Texture
   sortPriority=2
   texture=/Game/Global/Textures/Default/Noise/T_Tiling_Noise_01
  }
  n56 = LinearInterpolate(
   A=ScalarParameter() {defaultValue=1, group=Textures, parameterName=Opacity}
   B=StaticSwitchParameter(True=@n53.A, False=@n55.G) {parameterName=UseSubUV?}
   Alpha=ScalarParameter() {group=Textures, parameterName=UseOpacityMaskTexture?}
  )
  n57 = Add(A=@n21) {constB=-0.25}
  n58 = LinearInterpolate(Alpha=ConstantBiasScale(None=Sine(None=@n57)))
  StaticSwitchParameter(True=@n56, False=Multiply(A=@n56, B=Power(Base=@n58) {constExponent=0.6})) {defaultValue=true, parameterName=UseRotation}
  TextureSampleParameter2D() {
   group=Textures
   parameterName=Normal Texture
   samplerType=SAMPLERTYPE_Normal
   sortPriority=3
   texture=/Game/Environments/Props/Trash/Textures/Prop_Trash_N
  }
 ]
}
