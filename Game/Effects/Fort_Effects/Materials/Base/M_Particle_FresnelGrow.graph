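# Unlit, two-sided translucent particle material for mesh and sprite effects, drawing a fresnel glow tinted by particle color.
# Dynamic parameters drive fresnel exponent, multiplier and mesh inflation along vertex normals, with a static switch that inverts the fresnel, on by default.
# A blendable depth-fade term, optionally uninverted, adds intersection glow to emissive and opacity.
# Emissive passes through MF_FX_EmissiveAdjustments, and an optional squared near-camera fade scales opacity.
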Material /Game/Effects/Fort_Effects/Materials/Base/M_Particle_FresnelGrow
graph {
 properties {
  bUsedWithMeshParticles=true
  bUsedWithNiagaraMeshParticles=true
  bUsedWithParticleSprites=true
  blendMode=BLEND_Translucent
  pixelDepthOffsetMode=PDOM_Legacy
  shadingModel=MSM_Unlit
  twoSided=true
  usedShadingModels=Unlit
 }
 nodes [
  n1 = ParticleColor()
  n2 = DynamicParameter() {defaultValue=<rgba 0 5 1 0>, paramNames [InflateMesh,FresnelExp,FresnelMult,na]}
  n3 = Saturate(None=Multiply(A=Fresnel(ExponentIn=@n2.FresnelExp) {baseReflectFraction=0}, B=@n2.FresnelMult))
  n4 = StaticSwitchParameter(True=OneMinus(None=@n3), False=@n3) {defaultValue=true, parameterName=InvertFresnel}
  n5 = Multiply(A=@n1.RGB, B=@n4)
  n6 = DepthFade(Opacity=Constant() {r=1}, FadeDistance=ScalarParameter() {defaultValue=20, parameterName=FadeDistance})
  n7 = OneMinus(None=@n6)
  n8 = ScalarParameter() {parameterName=1ForUninvertedDepthFade, sliderMax=1}
  n9 = Multiply(A=Clamp(None=LinearInterpolate(A=@n7, B=@n6, Alpha=@n8)), B=ScalarParameter() {group=AdditiveDepthFade, parameterName=AdditiveGlowIntensity})
  n10 = Multiply(A=@n9, B=VectorParameter() {parameterName=EdgeIntensity, defaultValue=<rgba 1 1 1 1>}.RGB)
  MaterialFunctionCall(Emissive=LinearInterpolate(A=@n5, B=Add(A=@n5, B=@n10), Alpha=ScalarParameter() {parameterName=AdditiveDepthFade?})) {materialFunction=&MF_FX_EmissiveAdjustments}
  Multiply(A=VertexNormalWS(), B=DynamicParameter() {defaultValue=<rgba 0 5 1 0>, paramNames [InflateMesh,FresnelExp,FresnelMult,na]}.InflateMesh)
  n11 = Multiply(A=@n1.A, B=@n4)
  n12 = ScalarParameter() {defaultValue=1, parameterName=EdgeOpacity}
  n13 = LinearInterpolate(A=@n11, B=Add(A=@n11, B=Multiply(A=@n9, B=@n12)), Alpha=ScalarParameter() {parameterName=AdditiveDepthFade?})
  n14 = MaterialFunctionCall(Distance=ScalarParameter() {defaultValue=250, parameterName=NearFadeDistance}) {materialFunction=&Near_Camera_Fade}
  StaticSwitchParameter(True=Multiply(A=@n13, B=Multiply(A=@n14.Result, B=@n14.Result)), False=@n13) {parameterName=Use_CameraFade}
 ]
}
